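# Physics asset for the Horned Mask commando backpack cosmetic, defining bodies and joints for its dangling parts.
# Box and capsule bodies cover the bag sections, mask, scroll, loops, pouch and two cords, with most branches constrained to the offsetAll body.
# Cord A forms a five-link chain, cord B runs through the pouch, and the scroll hangs off the upper bag.
# Bag sections and cord A's root rotate freely; other joints use limited or locked swing and twist.
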PhysicsAsset /BRCosmetics/Accessories/FORT_Backpacks/F_MED_Commando_HornedMask_Backpack/Meshes/F_MED_Horned_Mask_Backpack_Physics
bodies {
 dyn_bagLower [
  {shapeType=Box, extentX=9.9897, extentY=8.4486, extentZ=6.9893, center=<-2.4768 -1.3036 9.7446>, rotation=<rot 5.9664 -89.675 8.8922>}
  {shapeType=Box, extentX=7.5, extentY=7.5, extentZ=7.5, center=<-4.1536 -4.5092 3.7443>}
  {shapeType=Box, extentX=7.5, extentY=7.5, extentZ=7.5, center=<0.2118 -4.6555 3.0989>, rotation=<rot -0.3592 10.1387 3.697>}
  {shapeType=Box, extentX=7.5, extentY=7.5, extentZ=8.5, center=<4.3375 -2.6137 7.626>, rotation=<rot -8.2996 11.3296 8.6399>}
 ]
 dyn_bagMain [
  {shapeType=Box, extentX=9, extentY=2.5, extentZ=9, center=<2.3296 -7.7768 5.2828>, rotation=<rot 18.1485 -95.401 -43.6647>}
  {shapeType=Box, extentX=11, extentY=21, extentZ=5.5, center=<-4.1456 -7.4303 -8.8786>, rotation=<rot 16.9215 10.6681 11.1762>}
 ]
 dyn_bagUpper [{shapeType=Capsule, length=6.5, radius=6, center=<-6.4436 -2.4251 0.0963>, rotation=<rot 4.9869 -90.7361 -1.0964>}]
 dyn_cordA1 [{shapeType=Capsule, length=3.7697, radius=0.7773, center=<3.3924 0.502 -0.0731>, rotation=<rot 23.1376 -73.9539 -81.1813>}]
 dyn_cordA2 [{shapeType=Capsule, length=4, radius=0.8254, center=<3 0.2968 0.1602>, rotation=<rot -0.3059 -85.9202 -87.2314>}]
 dyn_cordA3 [{shapeType=Capsule, length=5, radius=1, shapeName=dyn_cordA3_capsule, center=<4 0 0>, rotation=<rot 90 0 0>}]
 dyn_cordA4 [{shapeType=Capsule, length=5, radius=1, shapeName=dyn_cordA4_capsule, center=<3 0 0>, rotation=<rot 90 0 0>}]
 dyn_cordA5 [{shapeType=Capsule, length=3.6563, radius=2.25, center=<4.7619 -0.4721 -0.4963>, rotation=<rot 0.1135 82.4556 -96.4022>}]
 dyn_cordB2 [{shapeType=Capsule, length=4, radius=1.25, shapeName=dyn_cordB2_capsule, center=<4 0.5 -0.5>, rotation=<rot 85 0 -5>}]
 dyn_cordB3 [{shapeType=Capsule, length=3, radius=1.25, shapeName=dyn_cordB3_capsule, center=<3 0.25 -0.5>, rotation=<rot 80 0 -5>}]
 dyn_cordB4 [{shapeType=Capsule, length=2.7822, radius=2, center=<3.6253 0.1941 -0.4713>, rotation=<rot 1.5382 91.2118 -96.2472>}]
 dyn_cordB6 [{shapeType=Capsule, length=8, radius=1.5, shapeName=dyn_cordB6_capsule, center=<6.6428 -1.2004 1.0937>, rotation=<rot 82 180 -180>}]
 dyn_loopA2 [{
   shapeType=Capsule
   length=4.8532
   radius=2.1402
   shapeName=dyn_loopA2_capsule
   center=<7.9207 -2.0802 -7.8907>
   rotation=<rot 32.7881 -90 5.101>
  }]
 dyn_loopB1 [{shapeType=Box, extentX=3.9723, extentY=1.7873, extentZ=5.4105, center=<1.3424 2.0503 -1.3548>, rotation=<rot 0 -18.1519 -55.9192>}]
 dyn_loopB2 [{shapeType=Capsule, length=3, radius=2, shapeName=dyn_loopB2_capsule, center=<-3.9399 1.3834 -4.6798>, rotation=<rot -9.3913 3.4512 -20.2838>}]
 dyn_mask [{shapeType=Capsule, length=12.5114, radius=4.4716, center=<0 -2.9832 -12.355>, rotation=<rot 4.811 -89.9999 0>}]
 dyn_pouch [
  {shapeType=Capsule, length=1, radius=4.5, shapeName=dyn_pouch_capsule, center=<9 0 0>}
  {shapeType=Capsule, length=1, radius=3, shapeName=dyn_pouch_capsule, center=<3 0 0>, rotation=<rot 90 0 0>}
 ]
 dyn_scroll [{shapeType=Capsule, length=17.3472, radius=3.8365, center=<0.0292 -0.2942 0.6671>, rotation=<rot 0 -90.4011 -89.9861>}]
 offsetAll [{shapeType=Box, extentX=8.7089, extentY=18.8436, extentZ=22.4807, center=<-5.681 -9.2665 0.8667>, rotation=<rot 8.3383 -80.119 -88.6654>}]
}
constraints [
 {
  bone1Name=dyn_cordA2
  bone2Name=dyn_cordA1
  swing1LimitDegrees=5
  swing1Motion=Locked
  swing2LimitDegrees=5
  swing2Motion=Locked
  twistLimitDegrees=45
  twistMotion=Locked
 }
 {bone1Name=dyn_cordA4, bone2Name=dyn_cordA3, swing1LimitDegrees=12, swing2LimitDegrees=22, twistLimitDegrees=3}
 {bone1Name=dyn_cordA5, bone2Name=dyn_cordA4, swing1LimitDegrees=22, swing2LimitDegrees=22, twistLimitDegrees=45, twistMotion=Locked}
 {bone1Name=dyn_pouch, bone2Name=dyn_cordB6, swing1LimitDegrees=10, swing2LimitDegrees=10, twistLimitDegrees=45, twistMotion=Locked}
 {bone1Name=dyn_cordB2, bone2Name=dyn_pouch, swing1LimitDegrees=20, swing2LimitDegrees=10, twistLimitDegrees=45}
 {bone1Name=dyn_cordB3, bone2Name=dyn_cordB2, swing1LimitDegrees=22, swing2LimitDegrees=22, twistLimitDegrees=45}
 {bone1Name=dyn_cordB4, bone2Name=dyn_cordB3, swing1LimitDegrees=45, swing2LimitDegrees=45, twistLimitDegrees=12}
 {bone1Name=dyn_scroll, bone2Name=dyn_bagUpper, swing1LimitDegrees=5, swing2LimitDegrees=2, twistLimitDegrees=1}
 {bone1Name=dyn_cordA3, bone2Name=dyn_cordA2, swing1LimitDegrees=3, swing2LimitDegrees=22, twistLimitDegrees=3}
 {bone1Name=dyn_mask, bone2Name=offsetAll, swing1LimitDegrees=2, swing1Motion=Locked, swing2LimitDegrees=2, twistLimitDegrees=22}
 {
  bone1Name=dyn_cordA1
  bone2Name=offsetAll
  swing1LimitDegrees=45
  swing1Motion=Free
  swing2LimitDegrees=45
  swing2Motion=Free
  twistLimitDegrees=45
  twistMotion=Free
 }
 {
  bone1Name=dyn_cordB6
  bone2Name=offsetAll
  swing1LimitDegrees=3
  swing1Motion=Locked
  swing2LimitDegrees=3
  twistLimitDegrees=45
  twistMotion=Locked
 }
 {bone1Name=dyn_loopB1, bone2Name=offsetAll, swing1LimitDegrees=6, swing2LimitDegrees=45, swing2Motion=Locked, twistLimitDegrees=12}
 {bone1Name=dyn_loopB2, bone2Name=offsetAll, swing1LimitDegrees=6, swing2LimitDegrees=22, twistLimitDegrees=45, twistMotion=Locked}
 {
  bone1Name=dyn_bagUpper
  bone2Name=offsetAll
  swing1LimitDegrees=45
  swing1Motion=Free
  swing2LimitDegrees=45
  swing2Motion=Free
  twistLimitDegrees=45
  twistMotion=Free
 }
 {
  bone1Name=dyn_bagMain
  bone2Name=offsetAll
  swing1LimitDegrees=45
  swing1Motion=Free
  swing2LimitDegrees=45
  swing2Motion=Free
  twistLimitDegrees=45
  twistMotion=Free
 }
 {
  bone1Name=dyn_bagLower
  bone2Name=offsetAll
  swing1LimitDegrees=45
  swing1Motion=Free
  swing2LimitDegrees=45
  swing2Motion=Free
  twistLimitDegrees=45
  twistMotion=Free
 }
 {bone1Name=dyn_loopA2, bone2Name=offsetAll, swing1LimitDegrees=45, swing2LimitDegrees=45, twistLimitDegrees=0}
]
