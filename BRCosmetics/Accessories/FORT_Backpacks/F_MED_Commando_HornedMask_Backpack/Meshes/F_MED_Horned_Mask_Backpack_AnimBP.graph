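# Animation Blueprint for the Horned Mask backpack cosmetic, adding secondary physics motion on top of an input pose.
# A RigidBody node simulates the backpack physics asset relative to backpack_root with scaled, clamped component motion, then feeds an Anim Dynamics node that swings dyn_scrollStick.
# A final additive Modify Bone nudges backpack_root slightly along its local Y axis.
# The event graph holds only unconnected default nodes.
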AnimBlueprint /BRCosmetics/Accessories/FORT_Backpacks/F_MED_Commando_HornedMask_Backpack/Meshes/F_MED_Horned_Mask_Backpack_AnimBP
target_skeleton=&/F_MED_Horned_Mask_Backpack_Skeleton
graphs [
 AnimGraph {nodes [
   AnimGraphNode_SubInput_0 = Input Pose()
   n1 = RigidBody(ComponentPose=Local To Component(LocalPose=@AnimGraphNode_SubInput_0)) {
    bOverrideWorldGravity=true
    overridePhysicsAsset=&/F_MED_Horned_Mask_Backpack_Physics
    alphaBoolBlend {blendOption=Linear}
    baseBoneRef {boneName=backpack_root}
    componentAppliedLinearAccClamp=<900 900 900>
    componentLinearAccScale=<0.4 0.4 0.4>
    componentLinearVelScale=<0.5 0.5 0.5>
    overrideWorldGravity {z=-980}
   }
   n2 = Anim Dynamics - Bone: dyn_scrollStick(ComponentPose=@n1, Alpha=1, ExternalForce="0,0,0") {
    angularDampingOverride=0.999
    angularSpringConstant=5
    bGravityOverrideInSimSpace=true
    bOverrideAngularDamping=true
    bUseGravityOverride=true
    alphaBoolBlend {blendOption=Linear}
    boundBone {boneName=dyn_scrollStick}
    componentAppliedLinearAccClamp=<300 300 300>
    componentLinearAccScale=<1 1 1>
    componentLinearVelScale=<1 1 1>
    gravityOverride {z=-981}
    physicsBodyDefinitions [{
      sphereCollisionRadius=0
      boundBone {boneName=dyn_scrollStick}
      boxExtents=<16 5 5>
      constraintSetup {angularLimitsMax=<1 0.5 0.1>, angularLimitsMin=<1 0.5 0.1>, angularTarget=<0 0 0>}
      localJointOffset=<0 0 -2>
     }]
   }
   n3 = "Transform (Modify) Bone - Bone: backpack_root"(ComponentPose=@n2, Alpha=1, Translation="0,-1,0", Rotation="0,0,0", Scale="1,1,1") {
    translationMode=BMM_Additive
    translationSpace=BCS_BoneSpace
    alphaBoolBlend {blendOption=Linear}
    boneToModify {boneName=backpack_root}
    translation {y=-1}
   }
   Output Pose(Result=Component To Local(ComponentPose=@n3)) {layerGroup=None}
  ]}
 EventGraph {nodes [
   TryGetPawnOwner()
   Event BlueprintUpdateAnimation()
  ]}
]
variables []
